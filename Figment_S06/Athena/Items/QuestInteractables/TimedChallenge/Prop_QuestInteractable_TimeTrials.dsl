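; Quest interactable prop for Athena time trials: interacting starts a timed run through eight checkpoints, awarding quest progress on success.
; BeginPlay waits five seconds, then checks the playlist, disabling interaction in Creative and hiding the prop from clients without the open objective.
; Interaction records each runner in an array polled by a looping timer for timeouts.
; Clients animate checkpoints outward, then run countdown materials, audio and resets, with a safety timeout for lost relevancy.
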(event Custom|QuestInitializedAndValid
  ;[1 1 1] Decide if Object should be visible to player or not depending on if they have the quest and objective or not
  (Rendering|Components|Light|SetIndirectLightingIntensity (Variables|Default|GetAmbientLight) 3e+03 :!class /Script/Engine.LightComponent))

(event EventBeginPlay
  ;[1 0.5445 0.1188] Workaround - needs to be something better for ship
  (Utilities|FlowControl|Delay 5.0)
  (Utilities|FlowControl|Sequence
    (:then_0
      ;[0.3586 1 0.2914] Initialize CP's and Visual Clock Timer Based off of TimeTrialTime Chosen
      (bind _outxy (Fortnite|EvaluateCurveTableRow (Variables|Default|GetTimerNumberLookup)))
      (bind _output_get (Variables|Default|SetTimeTrialTime (+ _outxy (Variables|Default|GetTimerBuffer))))
      (Development|PrintString (Utilities|String|ToString(Float) _output_get) false false "(R=0.000000,G=0.660000,B=1.000000,A=1.000000)" 1e+01)
      (CallFunction|SetUpCPsandTimerVisuals))
    (:then_1
      ;[1 1 1] CLIENT AND SERVER - Ensure playlist is loaded
      (bind _asfort_game_state_athena (Utilities|Casting|CastToFortGameStateAthena (Game|GetGameState)))
      (if (Variables|/Script/FortniteGame.FortGameStateAthena|GetbPlaylistDataIsLoaded _asfort_game_state_athena)
        (if (GameplayTags|HasTag (Class|FortGameStateAthena|GetAthenaPlaylistContextTags _asfort_game_state_athena) "(TagName=\"Athena.Creative\")")
          (Variables|Default|SetCreativeMode true)
          (Networking|SwitchHasAuthority
            (:Remote
              ;[1 1 1] Decide if Object should be visible to player or not depending on if they have the quest and objective or not
              (CallFunction|HideQuestProp)
              (Variables|QuestStuff|SetisLocallyValid false)))
          (else
            (Variables|Default|SetCreativeMode false)
            (Networking|SwitchHasAuthority
              (:Authority
                ;[1 1 1] SERVER: Bind CP overlaps
                (CallFunction|BindCPOverlaps))
              (:Remote
                (bind (_questisvalid _questcompleted _objectivecompleted) (Default|AthenaCheckQuestandObjectiveCompletion (Utilities|Array|Get(acopy) (Fortnite|GetLocalFortPlayerControllers true)) (Variables|QuestStuff|GetQuestItem) (Variables|QuestStuff|GetQuestBackendObjectiveName)))
                (if (and _questisvalid (not _objectivecompleted))
                  _
                  (elif (Variables|Default|GetTestSetup)
                    _
                    (else
                      (CallFunction|HideQuestProp)
                      (Variables|QuestStuff|SetisLocallyValid false))))
                (Variables|QuestStuff|SetisLocallyValid true)
                (CallFunction|QuestInitializedandValid)
                ;[1 1 1] Bind Event for When Quest Gets Updated
                (bind _asfort_player_controller_athena (Utilities|Casting|CastToFortPlayerControllerAthena (Utilities|Array|Get(acopy) (Fortnite|GetLocalFortPlayerControllers true))))
                (Quests|BindEventtoOnQuestsUpdated (Quests|GetQuestManager _asfort_player_controller_athena "Invalid") (EventDispatchers|CreateEvent :!event Custom|ParentQuestUpdated))))))
        (else
          (Playlist|BindEventtoOnCurrentPlaylistLoaded _asfort_game_state_athena (delegate Custom|PlaylistLoaded)))))))

(event Custom|PlaylistLoaded
  (params (PlaylistName :type Name) (PlaylistContextTags :type Gameplay%20Tag%20Container%20Structure%20%28by%20ref%29))
  ;[1 1 1] CLIENT AND SERVER - Ensure playlist is loaded
  (if (GameplayTags|HasTag PlaylistContextTags "(TagName=\"Athena.Creative\")")
    (Variables|Default|SetCreativeMode true)
    (Networking|SwitchHasAuthority
      (:Remote
        ;[1 1 1] Decide if Object should be visible to player or not depending on if they have the quest and objective or not
        (CallFunction|HideQuestProp)
        (Variables|QuestStuff|SetisLocallyValid false)))
    (else
      (Variables|Default|SetCreativeMode false)
      (Networking|SwitchHasAuthority
        (:Authority
          ;[1 1 1] SERVER: Bind CP overlaps
          (CallFunction|BindCPOverlaps))
        (:Remote
          (bind (_questisvalid _questcompleted _objectivecompleted) (Default|AthenaCheckQuestandObjectiveCompletion (Utilities|Array|Get(acopy) (Fortnite|GetLocalFortPlayerControllers true)) (Variables|QuestStuff|GetQuestItem) (Variables|QuestStuff|GetQuestBackendObjectiveName)))
          (if (and _questisvalid (not _objectivecompleted))
            _
            (elif (Variables|Default|GetTestSetup)
              _
              (else
                (CallFunction|HideQuestProp)
                (Variables|QuestStuff|SetisLocallyValid false))))
          (Variables|QuestStuff|SetisLocallyValid true)
          (CallFunction|QuestInitializedandValid)
          ;[1 1 1] Bind Event for When Quest Gets Updated
          (bind _asfort_player_controller_athena (Utilities|Casting|CastToFortPlayerControllerAthena (Utilities|Array|Get(acopy) (Fortnite|GetLocalFortPlayerControllers true))))
          (Quests|BindEventtoOnQuestsUpdated (Quests|GetQuestManager _asfort_player_controller_athena "Invalid") (EventDispatchers|CreateEvent :!event Custom|ParentQuestUpdated)))))))

(event Custom|AwardQuest
  (params (PawnDoingTrial :type Fort%20Pawn%20Object%20Reference))
  ;[1 1 1] Award Quest Progress
  (Utilities|IsValid PawnDoingTrial
    (:Is%20Valid
      (bind (_runneristracked _runnerindex) (CallFunction|IsRunnerTracked PawnDoingTrial))
      (if _runneristracked
        (Utilities|Array|RemoveIndex (Variables|Default|GetRunners) _runnerindex)
        (bind _asfort_player_controller (Utilities|Casting|CastToFortPlayerController (Pawn|GetController PawnDoingTrial)))
        (Class|AthenaQuestFunctionLibrary|AthenaGiveQuestUpdatetoPlayers (Variables|QuestStuff|GetQuestItem) (Variables|QuestStuff|GetQuestBackendObjectiveName) (Variables|QuestStuff|GetObjectiveStat) (Utilities|Array|MakeArray _asfort_player_controller))
        (CallFunction|PlaySuccessFX _asfort_player_controller)))))

(event Custom|ParentQuestUpdated
  ;[1 1 1] Make Interactable Non-Locally Valid
  (if (Variables|QuestStuff|GetisLocallyValid)
    (bind (_questisvalid _questcompleted _objectivecompleted) (Default|AthenaCheckQuestandObjectiveCompletion (Utilities|Array|Get(acopy) (Fortnite|GetLocalFortPlayerControllers true)) (Variables|QuestStuff|GetQuestItem) (Variables|QuestStuff|GetQuestBackendObjectiveName)))
    (if (or _questcompleted _objectivecompleted)
      (Variables|QuestStuff|SetisLocallyValid false)
      ;[1 1 1] UnBind Event for When Quest Gets Updated
      (bind _asfort_player_controller_athena (Utilities|Casting|CastToFortPlayerControllerAthena (Utilities|Array|Get(acopy) (Fortnite|GetLocalFortPlayerControllers true))))
      (Quests|UnbindEventfromOnQuestsUpdated (Quests|GetQuestManager _asfort_player_controller_athena "Invalid") (EventDispatchers|CreateEvent :!event Custom|ParentQuestUpdated)))))

(event Custom|Check%20Runners%20Timers
  (bind _runners (Variables|Default|GetRunners))
  ;[1 1 1] HANDLE TIMER FOR CLIENTS
  ;;[1 1 1] Runner Timer Checking
  (for (_array_element _array_index) _runners
    (if (< (+ (.StartTimeTrialTime_14_55B494C04FD1018523FE6EB2EE476C09 (Utilities|Struct|BreakTimeTrialRunnerData _array_element)) (+ (Variables|Default|GetTimeTrialTime) (Variables|Default|GetServerGraceTime))) (Utilities|Time|GetGameTimeinSeconds))
      (CallFunction|FailureFX (Utilities|Struct|BreakTimeTrialRunnerData _array_element))
      (Utilities|Array|RemoveIndex _runners _array_index)
      (Development|PrintString "Failture" false false "(R=1.000000,G=0.364894,B=0.000000,A=1.000000)" 5.0)
      (else
        (Development|PrintString "CheckTime" false false "(R=1.000000,G=0.364894,B=0.000000,A=1.000000)" 1.0))))
  (if (<= (Utilities|Array|Length _runners) 0)
    (Utilities|Time|ClearandInvalidateTimerbyHandle (Variables|Default|GetTrialTimer))
    (Development|PrintString "Timer Cleared" false false "(R=1.000000,G=0.000000,B=0.705616,A=1.000000)" 5.0)))

(event Interaction|EventBlueprintOnInteract
  (params (InteractingPawn :type Fort%20Pawn%20Object%20Reference) (InteractionBeingAttempted :type EInteractionBeingAttempted%20Enum :default "FirstInteraction"))
  ;[1 1 1] INTERACTION COMPLETE: Get interacting Player and set up Struct for each client running the trial
  (bind _runneristracked (CallFunction|IsRunnerTracked InteractingPawn))
  (if (not _runneristracked)
    (bind _asfort_player_controller (Utilities|Casting|CastToFortPlayerController (Pawn|GetController InteractingPawn)))
    (Utilities|Array|Add (Variables|Default|GetRunners) (Utilities|Struct|MakeTimeTrialRunnerData _asfort_player_controller InteractingPawn (Utilities|Time|GetGameTimeinSeconds) (Utilities|Array|MakeArray false)))
    (CallFunction|ClientStartTrial _asfort_player_controller)
    ;[1 1 1] HANDLE TIMER FOR CLIENTS
    ;;[1 1 1] TRIAL TIMER SET HERE
    (if (not (Utilities|Time|IsValidTimerHandle (Variables|Default|GetTrialTimer)))
      (bind _returnvalue (Utilities|Time|SetTimerbyEvent (EventDispatchers|CreateEvent :!event Custom|Check%20Runners%20Timers) (Variables|Default|GetCheckRunnersTimeDelta) true))
      (Variables|Default|SetTrialTimer _returnvalue))))

(event Custom|ClientStartTrial
  (params (InPlayer :type Fort%20Player%20Controller%20Object%20Reference))
  ;[1 1 1] CLIENT: Begin Moving CP's into Place and Hide Interactable
  (Utilities|IsValid InPlayer
    (:Is%20Valid
      (Networking|SwitchHasAuthority
        (:Remote
          (CallFunction|MoveCPsIntoPlace)
          ;[0.9928 0 1] Activate Sound
          (Audio|PlaySoundatLocation (Variables|TimeTrialSounds|GetTriggeredSound) (Transformation|GetActorLocation))
          (Variables|Default|SetTimeTrialRunning true)
          (Rendering|Components|Light|SetIntensity (Variables|Default|GetAmbientLight) 0.0)
          (Components|Activation|Deactivate (Variables|Default|GetAmbientSound))
          (Components|Activation|Deactivate (Variables|Default|GetAmbientGodBeamFX)))))))

(event Custom|MoveCPsIntoPlace
  (bind _cpnumplacement (Variables|CPTriggering|GetCPNumPlacement))
  ;[1 1 1] CLIENT: Visually Put CP's into place after Activation
  (bind (_direction _movetime _clockshrinktime) (Timeline|CP1Movement PlayFromStart)
    (:Update
      (Transformation|SetRelativeLocation (Variables|Default|GetCheckpoint1) (Math|Vector|Lerp(Vector) "0, 0, 0" (Variables|CPInitialization|GetCP1%20Location) _movetime) false true)
      (Transformation|SetRelativeScale3D (Variables|Default|GetAlarmClock) (Math|Vector|Lerp(Vector) "3.000000,3.000000,3.000000" "0.001000,0.001000,0.001000" _clockshrinktime)))
    (:Finished
      (Development|SetHiddeninGame (Variables|Default|GetCheckpoint1) false false)
      (Development|SetHiddeninGame (Utilities|Select :Index (Variables|CPTriggering|GetCPNumPlacement) :Option%200 (Variables|Default|GetCheckpoint2) :Option%201 (Variables|Default|GetCheckpoint3) :Option%202 (Variables|Default|GetCheckpoint4) :Option%203 (Variables|Default|GetCheckpoint5) :Option%204 (Variables|Default|GetCheckpoint6) :Option%205 (Variables|Default|GetCheckpoint7) :Option%206 (Variables|Default|GetCheckpoint8)) false false)
      (Components|Activation|Activate (Utilities|Select :Index (Variables|CPTriggering|GetCPNumPlacement) :Option%200 (Variables|Default|GetP_TimedChallenge_Trail) :Option%201 (Variables|Default|GetP_TimedChallenge_Trail1) :Option%202 (Variables|Default|GetP_TimedChallenge_Trail2) :Option%203 (Variables|Default|GetP_TimedChallenge_Trail3) :Option%204 (Variables|Default|GetP_TimedChallenge_Trail4) :Option%205 (Variables|Default|GetP_TimedChallenge_Trail5) :Option%206 (Variables|Default|GetP_TimedChallenge_Trail6) :Option%207 (Variables|Default|GetP_TimedChallenge_Trail7)) false)
      (Audio|PlaySound2D "/Game/Athena/Sounds/Quests/TimedTrial/Athena_Quest_TimeTrial_CPStop_Cue.Athena_Quest_TimeTrial_CPStop_Cue" 1.0 (+ 1.0 (* (Variables|CPTriggering|GetCPNumPlacement) (Variables|TimeTrialSounds|GetCP%20Hit%20Pitch%20Mulitplier))))
      (bind (_direction_2 _movetime_2) (Timeline|CP2Movement PlayFromStart)
        (:Update
          (Transformation|SetRelativeLocation (Utilities|Select :Index (Variables|CPTriggering|GetCPNumPlacement) :Option%200 (Variables|Default|GetCheckpoint2) :Option%201 (Variables|Default|GetCheckpoint3) :Option%202 (Variables|Default|GetCheckpoint4) :Option%203 (Variables|Default|GetCheckpoint5) :Option%204 (Variables|Default|GetCheckpoint6) :Option%205 (Variables|Default|GetCheckpoint7) :Option%206 (Variables|Default|GetCheckpoint8)) (Math|Vector|Lerp(Vector) (Utilities|Select :Index (Variables|CPTriggering|GetCPNumPlacement) :Option%200 (Variables|CPInitialization|GetCP1%20Location) :Option%201 (Variables|CPInitialization|GetCP2%20Location) :Option%202 (Variables|CPInitialization|GetCP3%20Location) :Option%203 (Variables|CPInitialization|GetCP4%20Location) :Option%204 (Variables|CPInitialization|GetCP5%20Location) :Option%205 (Variables|CPInitialization|GetCP6%20Location) :Option%206 (Variables|CPInitialization|GetCP7%20Location) :Option%207 (Variables|CPInitialization|GetCP8Location)) (Utilities|Select :Index (+ (Variables|CPTriggering|GetCPNumPlacement) 1) :Option%200 (Variables|CPInitialization|GetCP1%20Location) :Option%201 (Variables|CPInitialization|GetCP2%20Location) :Option%202 (Variables|CPInitialization|GetCP3%20Location) :Option%203 (Variables|CPInitialization|GetCP4%20Location) :Option%204 (Variables|CPInitialization|GetCP5%20Location) :Option%205 (Variables|CPInitialization|GetCP6%20Location) :Option%206 (Variables|CPInitialization|GetCP7%20Location) :Option%207 (Variables|CPInitialization|GetCP8Location)) _movetime_2) false true))
        (:Finished
          (Math|Integer|IncrementInt _cpnumplacement
            (:%20%20
              (if (< _cpnumplacement 7)
                (CallFunction|InitializeNextCP)
                (else
                  ;[1 1 1] CLIENT: Once All CP's into place, turn on FX and Bind Overlaps
                  (Components|Activation|Activate (Variables|Default|GetP_TimedChallenge_InWorldLocation1) false)
                  (CallFunction|BindCPOverlaps)
                  (Variables|CPTriggering|SetNumCPsHit 0)
                  ;[0.9815 0.1329 1] Start Music
                  (Components|Activation|Activate (Variables|Default|GetAmbientSound) true))))))))))

(event Custom|InitializeNextCP
  (bind _cpnumplacement (Variables|CPTriggering|GetCPNumPlacement))
  ;[1 1 1] CLIENT: Visually Put CP's into place after Activation
  (Development|SetHiddeninGame (Utilities|Select :Index (Variables|CPTriggering|GetCPNumPlacement) :Option%200 (Variables|Default|GetCheckpoint2) :Option%201 (Variables|Default|GetCheckpoint3) :Option%202 (Variables|Default|GetCheckpoint4) :Option%203 (Variables|Default|GetCheckpoint5) :Option%204 (Variables|Default|GetCheckpoint6) :Option%205 (Variables|Default|GetCheckpoint7) :Option%206 (Variables|Default|GetCheckpoint8)) false false)
  (Components|Activation|Activate (Utilities|Select :Index (Variables|CPTriggering|GetCPNumPlacement) :Option%200 (Variables|Default|GetP_TimedChallenge_Trail) :Option%201 (Variables|Default|GetP_TimedChallenge_Trail1) :Option%202 (Variables|Default|GetP_TimedChallenge_Trail2) :Option%203 (Variables|Default|GetP_TimedChallenge_Trail3) :Option%204 (Variables|Default|GetP_TimedChallenge_Trail4) :Option%205 (Variables|Default|GetP_TimedChallenge_Trail5) :Option%206 (Variables|Default|GetP_TimedChallenge_Trail6) :Option%207 (Variables|Default|GetP_TimedChallenge_Trail7)) false)
  (Audio|PlaySound2D "/Game/Athena/Sounds/Quests/TimedTrial/Athena_Quest_TimeTrial_CPStop_Cue.Athena_Quest_TimeTrial_CPStop_Cue" 1.0 (+ 1.0 (* (Variables|CPTriggering|GetCPNumPlacement) (Variables|TimeTrialSounds|GetCP%20Hit%20Pitch%20Mulitplier))))
  (Timeline|CP2Movement PlayFromStart))

(event Custom|Bind%20CP%20Overlaps
  ;[1 1 1] Bind and Unbind CP overlaps
  (Utilities|FlowControl|Sequence
    (:then_0
      (Collision|BindEventtoOnComponentBeginOverlap (Variables|Default|GetCP1Collision) (EventDispatchers|CreateEvent :!event Custom|CP1%20Triggered))
      (Collision|BindEventtoOnComponentBeginOverlap (Variables|Default|GetCP2Collision) (EventDispatchers|CreateEvent :!event Custom|CP2%20Triggered))
      (Collision|BindEventtoOnComponentBeginOverlap (Variables|Default|GetCP3Colllision) (EventDispatchers|CreateEvent :!event Custom|CP3%20Triggered))
      (Collision|BindEventtoOnComponentBeginOverlap (Variables|Default|GetCP4Collision) (EventDispatchers|CreateEvent :!event Custom|CP4%20Triggered))
      (Collision|BindEventtoOnComponentBeginOverlap (Variables|Default|GetCP5Collision) (EventDispatchers|CreateEvent :!event Custom|CP5%20Triggered))
      (Collision|BindEventtoOnComponentBeginOverlap (Variables|Default|GetCP6Collision) (EventDispatchers|CreateEvent :!event Custom|CP6%20Triggered))
      (Collision|BindEventtoOnComponentBeginOverlap (Variables|Default|GetCP7Collision) (EventDispatchers|CreateEvent :!event Custom|CP7%20Triggered))
      (Collision|BindEventtoOnComponentBeginOverlap (Variables|Default|GetCP8Coll) (EventDispatchers|CreateEvent :!event Custom|CP8Triggered)))
    (:then_1
      (CallFunction|StartVisualTimerCountdown))))

(event Custom|Unbind%20CP%20Overlaps
  ;[1 1 1] Bind and Unbind CP overlaps
  (Utilities|FlowControl|Sequence
    (:then_0
      (Collision|UnbindEventfromOnComponentBeginOverlap (Variables|Default|GetCP1Collision) (EventDispatchers|CreateEvent :!event Custom|CP1%20Triggered)))
    (:then_1
      (Collision|UnbindEventfromOnComponentBeginOverlap (Variables|Default|GetCP2Collision) (EventDispatchers|CreateEvent :!event Custom|CP2%20Triggered)))
    (:then_2
      (Collision|UnbindEventfromOnComponentBeginOverlap (Variables|Default|GetCP3Colllision) (EventDispatchers|CreateEvent :!event Custom|CP3%20Triggered)))
    (:then_3
      (Collision|UnbindEventfromOnComponentBeginOverlap (Variables|Default|GetCP4Collision) (EventDispatchers|CreateEvent :!event Custom|CP4%20Triggered)))
    (:then_4
      (Collision|UnbindEventfromOnComponentBeginOverlap (Variables|Default|GetCP5Collision) (EventDispatchers|CreateEvent :!event Custom|CP5%20Triggered)))
    (:then_5
      (Collision|UnbindEventfromOnComponentBeginOverlap (Variables|Default|GetCP6Collision) (EventDispatchers|CreateEvent :!event Custom|CP6%20Triggered)))
    (:then_6
      (Collision|UnbindEventfromOnComponentBeginOverlap (Variables|Default|GetCP7Collision) (EventDispatchers|CreateEvent :!event Custom|CP7%20Triggered)))
    (:then_7
      (Collision|UnbindEventfromOnComponentBeginOverlap (Variables|Default|GetCP8Coll) (EventDispatchers|CreateEvent :!event Custom|CP8Triggered)))))

(event Custom|CP1%20Triggered
  (params (OverlappedComponent :type Primitive%20Component%20Object%20Reference) (OtherActor :type Actor%20Object%20Reference) (OtherComp :type Primitive%20Component%20Object%20Reference) (OtherBodyIndex :type Integer) (bFromSweep :type Boolean) (SweepResult :type Hit%20Result%20Structure%20%28by%20ref%29))
  ;[1 1 1] Handle Triggering of CPs
  (Utilities|CPTriggerCheck OtherActor (Variables|Default|GetCheckpoint1) (Variables|Default|GetP_TimedChallenge_InWorldLocation)
    (:ValidTrigger
      ;[1 1 1] Bind and Unbind CP overlaps
      (Collision|UnbindEventfromOnComponentBeginOverlap (Variables|Default|GetCP1Collision) (EventDispatchers|CreateEvent :!event Custom|CP1%20Triggered)))))

(event Custom|CP2%20Triggered
  (params (OverlappedComponent :type Primitive%20Component%20Object%20Reference) (OtherActor :type Actor%20Object%20Reference) (OtherComp :type Primitive%20Component%20Object%20Reference) (OtherBodyIndex :type Integer) (bFromSweep :type Boolean) (SweepResult :type Hit%20Result%20Structure%20%28by%20ref%29))
  ;[1 1 1] Handle Triggering of CPs
  (Utilities|CPTriggerCheck OtherActor (Variables|Default|GetCheckpoint2) (Variables|Default|GetP_TimedChallenge_InWorldLocation1) 1
    (:ValidTrigger
      ;[1 1 1] Bind and Unbind CP overlaps
      (Collision|UnbindEventfromOnComponentBeginOverlap (Variables|Default|GetCP2Collision) (EventDispatchers|CreateEvent :!event Custom|CP2%20Triggered)))))

(event Custom|CP3%20Triggered
  (params (OverlappedComponent :type Primitive%20Component%20Object%20Reference) (OtherActor :type Actor%20Object%20Reference) (OtherComp :type Primitive%20Component%20Object%20Reference) (OtherBodyIndex :type Integer) (bFromSweep :type Boolean) (SweepResult :type Hit%20Result%20Structure%20%28by%20ref%29))
  ;[1 1 1] Handle Triggering of CPs
  (Utilities|CPTriggerCheck OtherActor (Variables|Default|GetCheckpoint3) (Variables|Default|GetP_TimedChallenge_InWorldLocation2) 2
    (:ValidTrigger
      ;[1 1 1] Bind and Unbind CP overlaps
      (Collision|UnbindEventfromOnComponentBeginOverlap (Variables|Default|GetCP3Colllision) (EventDispatchers|CreateEvent :!event Custom|CP3%20Triggered)))))

(event Custom|CP4%20Triggered
  (params (OverlappedComponent :type Primitive%20Component%20Object%20Reference) (OtherActor :type Actor%20Object%20Reference) (OtherComp :type Primitive%20Component%20Object%20Reference) (OtherBodyIndex :type Integer) (bFromSweep :type Boolean) (SweepResult :type Hit%20Result%20Structure%20%28by%20ref%29))
  ;[1 1 1] Handle Triggering of CPs
  (Utilities|CPTriggerCheck OtherActor (Variables|Default|GetCheckpoint4) (Variables|Default|GetP_TimedChallenge_InWorldLocation3) 3
    (:ValidTrigger
      ;[1 1 1] Bind and Unbind CP overlaps
      (Collision|UnbindEventfromOnComponentBeginOverlap (Variables|Default|GetCP4Collision) (EventDispatchers|CreateEvent :!event Custom|CP4%20Triggered)))))

(event Custom|CP5%20Triggered
  (params (OverlappedComponent :type Primitive%20Component%20Object%20Reference) (OtherActor :type Actor%20Object%20Reference) (OtherComp :type Primitive%20Component%20Object%20Reference) (OtherBodyIndex :type Integer) (bFromSweep :type Boolean) (SweepResult :type Hit%20Result%20Structure%20%28by%20ref%29))
  ;[1 1 1] Handle Triggering of CPs
  (Utilities|CPTriggerCheck OtherActor (Variables|Default|GetCheckpoint5) (Variables|Default|GetP_TimedChallenge_InWorldLocation4) 4
    (:ValidTrigger
      ;[1 1 1] Bind and Unbind CP overlaps
      (Collision|UnbindEventfromOnComponentBeginOverlap (Variables|Default|GetCP5Collision) (EventDispatchers|CreateEvent :!event Custom|CP5%20Triggered)))))

(event Custom|CP6%20Triggered
  (params (OverlappedComponent :type Primitive%20Component%20Object%20Reference) (OtherActor :type Actor%20Object%20Reference) (OtherComp :type Primitive%20Component%20Object%20Reference) (OtherBodyIndex :type Integer) (bFromSweep :type Boolean) (SweepResult :type Hit%20Result%20Structure%20%28by%20ref%29))
  ;[1 1 1] Handle Triggering of CPs
  (Utilities|CPTriggerCheck OtherActor (Variables|Default|GetCheckpoint6) (Variables|Default|GetP_TimedChallenge_InWorldLocation5) 5
    (:ValidTrigger
      ;[1 1 1] Bind and Unbind CP overlaps
      (Collision|UnbindEventfromOnComponentBeginOverlap (Variables|Default|GetCP6Collision) (EventDispatchers|CreateEvent :!event Custom|CP6%20Triggered)))))

(event Custom|CP7%20Triggered
  (params (OverlappedComponent :type Primitive%20Component%20Object%20Reference) (OtherActor :type Actor%20Object%20Reference) (OtherComp :type Primitive%20Component%20Object%20Reference) (OtherBodyIndex :type Integer) (bFromSweep :type Boolean) (SweepResult :type Hit%20Result%20Structure%20%28by%20ref%29))
  ;[1 1 1] Handle Triggering of CPs
  (Utilities|CPTriggerCheck OtherActor (Variables|Default|GetCheckpoint7) (Variables|Default|GetP_TimedChallenge_InWorldLocation6) 6
    (:ValidTrigger
      ;[1 1 1] Bind and Unbind CP overlaps
      (Collision|UnbindEventfromOnComponentBeginOverlap (Variables|Default|GetCP7Collision) (EventDispatchers|CreateEvent :!event Custom|CP7%20Triggered)))))

(event Custom|CP8Triggered
  (params (OverlappedComponent :type Primitive%20Component%20Object%20Reference) (OtherActor :type Actor%20Object%20Reference) (OtherComp :type Primitive%20Component%20Object%20Reference) (OtherBodyIndex :type Integer) (bFromSweep :type Boolean) (SweepResult :type Hit%20Result%20Structure%20%28by%20ref%29))
  ;[1 1 1] Handle Triggering of CPs
  (Utilities|CPTriggerCheck OtherActor (Variables|Default|GetCheckpoint8) (Variables|Default|GetP_TimedChallenge_InWorldLocations7) 7
    (:ValidTrigger
      ;[1 1 1] Bind and Unbind CP overlaps
      (Collision|UnbindEventfromOnComponentBeginOverlap (Variables|Default|GetCP8Coll) (EventDispatchers|CreateEvent :!event Custom|CP8Triggered)))))

(event Custom|Start%20Visual%20Timer%20Countdown
  (bind _checkpoint1 (Variables|Default|GetCheckpoint1))
  ;[1 1 1] Visual Timer on CP's
  ;;[1 1 1] Safety In case player drops out of net relevancey
  (bind _returnvalue (Utilities|Time|SetTimerbyEvent (delegate Custom|SafetyTimeOut) (+ (Variables|Default|GetTimeTrialTime) 2.0)))
  (Variables|Default|SetSafetyTimer _returnvalue)
  (bind (_direction _fadetime _bursttime) (Timeline|TIMER%20VISUALS PlayFromStart)
    (:Update
      (Rendering|Material|SetScalarParameterValueonMaterials _checkpoint1 "TimeLeft" (Math|Float|Lerp 1.0 0.0 _fadetime))
      (Rendering|Material|SetScalarParameterValueonMaterials _checkpoint1 "Brightness" (Math|Float|Lerp 1.0 1e+01 _bursttime)))
    (:NotifyEnding
      ;[0.8456 0 1] FinalSecondsSound
      (Audio|PlaySound2D (Variables|TimeTrialSounds|GetTimerFinalSecondsSound)))))

(event Custom|End%20Visual%20Timer%20Countdown
  (bind _checkpoint1 (Variables|Default|GetCheckpoint1))
  ;[1 1 1] Visual Timer on CP's
  (Timeline|TIMER%20VISUALS Stop))

(event Custom|PlaySuccessFX
  (params (Controller :type Fort%20Player%20Controller%20Object%20Reference))
  (bind _returnvalue (Game|Player|GetPlayerPawn Controller))
  (bind _questsuccesssound (Variables|QuestSounds&FX|GetQuestSuccessSound))
  (bind _questsuccessfx (Variables|QuestSounds&FX|GetQuestSuccessFX))
  ;[1 1 1] SUCCESS: FX and Sounds
  (Networking|SwitchHasAuthority
    (:Remote
      (Utilities|IsValid Controller
        (:Is%20Valid
          (Utilities|FlowControl|Sequence
            (:then_0
              ;[1 1 1] Visual Timer on CP's
              ;;[1 1 1] Safety In case player drops out of net relevancey
              (Utilities|Time|ClearandInvalidateTimerbyHandle (Variables|Default|GetSafetyTimer)))
            (:then_1
              (CallFunction|EndVisualTimerCountdown)
              ;[0.8456 0 1] FailSound
              (Audio|PlaySound2D (Variables|TimeTrialSounds|GetTimeTrialSuccessSound))
              (CallFunction|ResetTimeTrial))
            (:then_2
              (Utilities|FlowControl|Delay 1.5)
              (Utilities|IsValid _questsuccesssound
                (:Is%20Valid
                  (Audio|SpawnSoundatLocation _questsuccesssound (Transformation|GetActorLocation _returnvalue)))
                (:Is%20Not%20Valid
                  _))
              (Utilities|IsValid _questsuccessfx
                (:Is%20Valid
                  (Effects|SpawnFXSystemAttached _questsuccessfx (Variables|/Script/Engine.Character|GetMesh _returnvalue)))))))))))

(event Custom|SafetyTimeOut
  ;[1 1 1] FAILURE:  FX and Sound
  (Networking|SwitchHasAuthority
    (:Remote
      (CallFunction|ResetTimeTrial)
      (Development|SetHiddeninGame (Variables|Default|GetCheckpoint1) true false)
      (Components|Activation|Deactivate (Variables|Default|GetP_TimedChallenge_InWorldLocation)))))

(event Custom|Failure%20FX
  (params (Player :type Fort%20Player%20Controller%20Object%20Reference))
  ;[1 1 1] FAILURE:  FX and Sound
  (Development|PrintString "FAILURE" false false "(R=1.000000,G=0.071664,B=0.000000,A=1.000000)" 5.0)
  (Utilities|IsValid Player
    (:Is%20Valid
      (Networking|SwitchHasAuthority
        (:Remote
          (Utilities|Time|ClearandInvalidateTimerbyHandle (Variables|Default|GetSafetyTimer))
          ;[1 1 1] CLIENT CLEAN UP
          (CallFunction|ResetTimeTrial)
          (Development|SetHiddeninGame (Variables|Default|GetCheckpoint1) true false)
          (Components|Activation|Deactivate (Variables|Default|GetP_TimedChallenge_InWorldLocation))
          (Audio|PlaySound2D (Variables|TimeTrialSounds|GetTimeTrialFailureSound)))))
    (:Is%20Not%20Valid
      (Development|PrintString "Player Not Valid" false false "(R=1.000000,G=0.071664,B=0.000000,A=1.000000)" 5.0))))

(event Custom|Reset%20Time%20Trial
  ;[1 1 1] RESET TIME TRIAL - CLIENT
  ;[0.9038 0 1] End Music
  (Components|Activation|Deactivate (Variables|Default|GetAmbientSound))
  (Variables|CPTriggering|SetCPNumPlacement 0)
  (CallFunction|UnbindCPOverlaps)
  (Utilities|FlowControl|Delay 0.5)
  (if (Variables|QuestStuff|GetisLocallyValid)
    (Development|PrintString "RESET " false false "(R=0.332427,G=1.000000,B=0.000000,A=1.000000)" 5.0)
    (Development|SetHiddeninGame (Variables|Default|GetAlarmClock) false false)
    (Rendering|Components|Light|SetIntensity (Variables|Default|GetAmbientLight) 3e+03)
    (Components|Activation|Activate (Variables|Default|GetAmbientGodBeamFX) false)
    (bind (_direction _movetime _clockshrinktime) (Timeline|Timeline_0 PlayFromStart)
      (:Update
        (Transformation|SetRelativeScale3D (Variables|Default|GetAlarmClock) (Math|Vector|Lerp(Vector) "0.000100,0.000100,0.001000" "3.000000,3.000000,3.000000" _clockshrinktime)))
      (:Finished
        (Variables|Default|SetTimeTrialRunning false)))
    (else
      (Development|PrintString "Not Locally Valid Anymore" false false "(R=1.000000,G=0.019224,B=0.000000,A=1.000000)" 5.0)
      (CallFunction|HideQuestProp))))

(fn BlueprintCanInteract
  (params (InteractingPawn :type Fort%20Pawn%20Object%20Reference) (InteractionBeingAttempted :type EInteractionBeingAttempted%20Enum :default "FirstInteraction") (InteractionType :type TInteractionType%20Enum :default "IT_NoInteraction"))
  (Utilities|IsValid InteractingPawn
    (:Is%20Valid
      (bind _asfort_player_controller_athena (Utilities|Casting|CastToFortPlayerControllerAthena (Game|GetInstigatorController InteractingPawn)))
      (bind (_questisvalid _questcompleted _objectivecompleted) (Default|AthenaCheckQuestandObjectiveCompletion _asfort_player_controller_athena (Variables|QuestStuff|GetQuestItem) (Variables|QuestStuff|GetQuestBackendObjectiveName)))
      (if (and _questisvalid (not _objectivecompleted))
        (LocalVariables|SetQuestValid true)
        (else
          (LocalVariables|SetQuestValid false)))
      (return (and (and (or (LocalVariables|GetQuestValid) (Variables|Default|GetTestSetup)) (not (Variables|Default|GetTimeTrialRunning))) (not (Variables|Default|GetCreativeMode)))))))

(fn BlueprintGetInteractionString
  (params (InteractingPawn :type Fort%20Pawn%20Object%20Reference) (InteractionBeingAttempted :type EInteractionBeingAttempted%20Enum :default "FirstInteraction"))
  (return (Variables|Default|GetInteractText)))

(fn CheckTrialRunFinished
  (params (RunningPawn :type Fort%20Pawn%20Object%20Reference))
  (bind _runners (Variables|Default|GetRunners))
  (bind (_runneristracked _runnerindex) (CallFunction|IsRunnerTracked RunningPawn))
  (if _runneristracked
    (if (>= (Utilities|Array|Length (.CPTriggered_44_4B52E41B47DCC48019DAB08129A9B3C8 (Utilities|Struct|BreakTimeTrialRunnerData (Utilities|Array|Get(aref) _runners _runnerindex)))) (Variables|CPTriggering|GetTotalCPNumber))
      (LocalVariables|SetAllCompleted true)
      (bind _array_element (Utilities|Array|ForEachLoopwithBreak (.CPTriggered_44_4B52E41B47DCC48019DAB08129A9B3C8 (Utilities|Struct|BreakTimeTrialRunnerData (Utilities|Array|Get(aref) _runners _runnerindex))))
        (:LoopBody
          (if (not _array_element)
            (LocalVariables|SetAllCompleted false)))
        (:Completed
          _))
      (if (LocalVariables|GetAllCompleted)
        (Development|PrintString "BOOL WORKS" false false "(R=1.000000,G=0.660563,B=0.093690,A=1.000000)" 5.0)
        (CallFunction|AwardQuest RunningPawn)
        (else
          (Development|PrintString "ALL CPs NOT HIT" false false "(R=1.000000,G=0.199603,B=0.000000,A=1.000000)" 5.0))))
    (else
      (Development|PrintString "Runner Not Tracked" false false "(R=1.000000,G=0.169153,B=0.000000,A=1.000000)" 5.0))))

(fn GetFocusedSocketLocation
  (return (Transformation|GetWorldLocation (Variables|Default|GetInteractPlacement))))

(fn HideQuestProp
  (Rendering|SetActorHiddenInGame true)
  (Rendering|Components|Light|SetIntensity (Variables|Default|GetAmbientLight) 0.0)
  (Components|Activation|Deactivate (Variables|Default|GetAmbientSound))
  (Components|Activation|Deactivate (Variables|Default|GetAmbientGodBeamFX)))

(fn IsRunnerTracked
  (params (InPawn :type Fort%20Pawn%20Object%20Reference))
  (bind _runners (Variables|Default|GetRunners))
  (if (== (Utilities|Array|Length _runners) 0)
    (return false 0)
    (else
      (bind (_array_element _array_index) (Utilities|Array|ForEachLoopwithBreak _runners)
        (:LoopBody
          (if (Utilities|Equal(Object) InPawn (.RunningPawn_15_6D6AFF35415544E8F16E49922C3918B0 (Utilities|Struct|BreakTimeTrialRunnerData _array_element)))
            (LocalVariables|SetFoundPassedInPawn true)
            (LocalVariables|SetFoundIndex _array_index)))
        (:Completed
          _))
      (return (LocalVariables|GetFoundPassedInPawn) (LocalVariables|GetFoundIndex)))))

(fn SetUpCPsAndTimerVisuals
  (bind _checkpoint8 (Variables|Default|GetCheckpoint8))
  (bind _checkpoint2 (Variables|Default|GetCheckpoint2))
  (bind _checkpoint3 (Variables|Default|GetCheckpoint3))
  (bind _checkpoint1 (Variables|Default|GetCheckpoint1))
  (bind _checkpoint4 (Variables|Default|GetCheckpoint4))
  (bind _checkpoint5 (Variables|Default|GetCheckpoint5))
  (bind _checkpoint6 (Variables|Default|GetCheckpoint6))
  (bind _checkpoint7 (Variables|Default|GetCheckpoint7))
  ;[1 1 1] Set up for Checkpoint Markers
  (Utilities|FlowControl|Sequence
    (:then_0
      (Variables|CPInitialization|SetCP1%20Location (Variables|/Script/Engine.SceneComponent|GetRelativeLocation _checkpoint1))
      (bind _returnvalue (Rendering|Material|CreateDynamicMaterialInstance _checkpoint1 1 "/Game/Effects/Fort_Effects/Materials/UI/M_TimeTrial_Flat.M_TimeTrial_Flat" "None" :!class /Script/Engine.PrimitiveComponent))
      (Variables|CPInitialization|SetCP1Mat _returnvalue)
      (Variables|CPInitialization|SetCP2%20Location (Variables|/Script/Engine.SceneComponent|GetRelativeLocation _checkpoint2))
      (bind _returnvalue_2 (Rendering|Material|CreateDynamicMaterialInstance _checkpoint2 1 "/Game/Effects/Fort_Effects/Materials/UI/M_TimeTrial_Flat.M_TimeTrial_Flat" "None" :!class /Script/Engine.PrimitiveComponent))
      (Variables|CPInitialization|SetCP2Mat _returnvalue_2)
      (Variables|CPInitialization|SetCP3%20Location (Variables|/Script/Engine.SceneComponent|GetRelativeLocation _checkpoint3))
      (bind _returnvalue_3 (Rendering|Material|CreateDynamicMaterialInstance _checkpoint3 1 "/Game/Effects/Fort_Effects/Materials/UI/M_TimeTrial_Flat.M_TimeTrial_Flat" "None" :!class /Script/Engine.PrimitiveComponent))
      (Variables|CPInitialization|SetCP3Mat _returnvalue_3)
      (Variables|CPInitialization|SetCP4%20Location (Variables|/Script/Engine.SceneComponent|GetRelativeLocation _checkpoint4))
      (bind _returnvalue_4 (Rendering|Material|CreateDynamicMaterialInstance _checkpoint4 1 "/Game/Effects/Fort_Effects/Materials/UI/M_TimeTrial_Flat.M_TimeTrial_Flat" "None" :!class /Script/Engine.PrimitiveComponent))
      (Variables|CPInitialization|SetCP4Mat _returnvalue_4)
      (Variables|CPInitialization|SetCP5%20Location (Variables|/Script/Engine.SceneComponent|GetRelativeLocation _checkpoint5))
      (bind _returnvalue_5 (Rendering|Material|CreateDynamicMaterialInstance _checkpoint5 1 "/Game/Effects/Fort_Effects/Materials/UI/M_TimeTrial_Flat.M_TimeTrial_Flat" "None" :!class /Script/Engine.PrimitiveComponent))
      (Variables|CPInitialization|SetCP5Mat _returnvalue_5)
      (Variables|CPInitialization|SetCP6%20Location (Variables|/Script/Engine.SceneComponent|GetRelativeLocation _checkpoint6))
      (bind _returnvalue_6 (Rendering|Material|CreateDynamicMaterialInstance _checkpoint6 1 "/Game/Effects/Fort_Effects/Materials/UI/M_TimeTrial_Flat.M_TimeTrial_Flat" "None" :!class /Script/Engine.PrimitiveComponent))
      (Variables|CPInitialization|SetCP6Mat _returnvalue_6)
      (Variables|CPInitialization|SetCP7%20Location (Variables|/Script/Engine.SceneComponent|GetRelativeLocation _checkpoint7))
      (bind _returnvalue_7 (Rendering|Material|CreateDynamicMaterialInstance _checkpoint7 1 "/Game/Effects/Fort_Effects/Materials/UI/M_TimeTrial_Flat.M_TimeTrial_Flat" "None" :!class /Script/Engine.PrimitiveComponent))
      (Variables|CPInitialization|SetCP7Mat _returnvalue_7)
      (Variables|CPInitialization|SetCP8Location (Variables|/Script/Engine.SceneComponent|GetRelativeLocation _checkpoint8))
      (bind _returnvalue_8 (Rendering|Material|CreateDynamicMaterialInstance _checkpoint8 1 "/Game/Effects/Fort_Effects/Materials/UI/M_TimeTrial_Flat.M_TimeTrial_Flat" "None" :!class /Script/Engine.PrimitiveComponent))
      (Variables|CPInitialization|SetCP8Mat _returnvalue_8))
    (:then_1
      ;[1 1 1] Initialize Play Rate for Clock Timer Visuals
      (Components|Timeline|SetPlayRate (Variables|Prop_QuestInteractable_TimeTrials|GetTIMER%20VISUALS) (/ 1.0 (/ (Variables|Default|GetTimeTrialTime) 1e+01))))))

(fn ConstructionScript)
